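# Physics asset for the F_MED_York female player body cosmetic, covering dangling hip gadgets, cables, a back clip and harness pieces.
# Capsules on pelvis, thighs and spine_04 form the body, with tiny base bodies linking accessories to it through free-rotation constraints.
# Short limited chains hang from those bases, with the box-shaped cableB rooted on the left thigh rather than pelvis.
# The back clip swings on one axis only, and harness pieces move very slightly.
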PhysicsAsset /BRCosmetics/Characters/Player/Female/Medium/Bodies/F_MED_York/Meshes/F_MED_York_Physics
bodies {
 dyn_harness_a_base_l [{shapeType=Capsule, length=0.505, radius=0.505, rotation=<rot 0 -90 0>}]
 dyn_harness_a_base_r [{shapeType=Capsule, length=0.505, radius=0.505, rotation=<rot 0 -90 180>}]
 dyn_harness_b_l [{shapeType=Capsule, length=4.4024, radius=1.0662, center=<3.3534 0 0>, rotation=<rot 90 0 0>}]
 dyn_harness_b_r [{shapeType=Capsule, length=4.4024, radius=1.0662, center=<-3.3534 0 0>, rotation=<rot 90 0 180>}]
 dyn_hip_backClip_a_base_r [{shapeType=Capsule, length=0.505, radius=0.505, rotation=<rot 0 -90 0>}]
 dyn_hip_backClip_a_r [{shapeType=Capsule, length=1.25, radius=1, center=<1.332 0 0>, rotation=<rot 90 0 0>}]
 dyn_hip_backClip_b_r [{shapeType=Capsule, length=2.25, radius=0.5, center=<1.6517 0 0>, rotation=<rot 90 0 0>}]
 dyn_hip_cableA_a_base_l [{shapeType=Capsule, length=0.505, radius=0.505, rotation=<rot 0 -90 0>}]
 dyn_hip_cableA_a_l [{shapeType=Capsule, length=5.5, radius=1.5, center=<2.5 0 0>, rotation=<rot 90 0 0>}]
 dyn_hip_cableA_b_l [{shapeType=Capsule, length=5.25, radius=1.75, center=<2.5 0 0>, rotation=<rot 90 0 0>}]
 dyn_hip_cableB_a_base_l [{shapeType=Capsule, length=0, radius=0}]
 dyn_hip_cableB_a_l [{shapeType=Box, extentX=11.5, extentY=1.8, extentZ=23, center=<6.1923 -0.8 0>, rotation=<rot 0.0002 0.0001 -10>}]
 dyn_hip_cableB_b_l [{shapeType=Box, extentX=12, extentY=1.8, extentZ=23, center=<6.3042 -0.25 -0.0001>, rotation=<rot 0.0006 0.0002 -10.0001>}]
 dyn_hip_gadget_a_base_l [{shapeType=Capsule, length=0.505, radius=0.505, rotation=<rot 0 -90 0>}]
 dyn_hip_gadget_a_base_r [{shapeType=Capsule, length=0.505, radius=0.505, rotation=<rot 0 -90 0>}]
 dyn_hip_gadget_a_l [{shapeType=Box, extentX=14.5, extentY=8.5, extentZ=3, center=<8.2451 0.0001 1.6049>}]
 dyn_hip_gadget_a_r [{shapeType=Capsule, length=21.25, radius=1.5, center=<8.4196 1.6687 0.1791>, rotation=<rot 88.7821 179.9974 179.9974>}]
 pelvis [{shapeType=Capsule, length=6, radius=10.5}, {shapeType=Capsule, length=9.5, radius=9.25, center=<-3.2049 -2.9187 0>}]
 spine_04 [{shapeType=Capsule, length=2.9257, radius=10.614, center=<-1.1503 2.0431 0>, rotation=<rot 0.1013 -90 0.3417>}]
 thigh_l [{shapeType=Capsule, length=30, radius=7.5, center=<-17.0934 0 0.0192>, rotation=<rot 87.9025 179.9999 179.9999>}]
 thigh_r [{shapeType=Capsule, length=30.25, radius=7, center=<17.0934 0 -0.0192>, rotation=<rot 87.9024 179.9999 -0.0001>}]
}
constraints [
 {
  bone1Name=thigh_l
  bone2Name=pelvis
  swing1LimitDegrees=45
  swing1Motion=Free
  swing2LimitDegrees=45
  swing2Motion=Free
  twistLimitDegrees=45
  twistMotion=Free
 }
 {
  bone1Name=thigh_r
  bone2Name=pelvis
  swing1LimitDegrees=45
  swing1Motion=Free
  swing2LimitDegrees=45
  swing2Motion=Free
  twistLimitDegrees=45
  twistMotion=Free
 }
 {
  bone1Name=dyn_hip_gadget_a_base_r
  bone2Name=pelvis
  swing1LimitDegrees=45
  swing1Motion=Free
  swing2LimitDegrees=45
  swing2Motion=Free
  twistLimitDegrees=45
  twistMotion=Free
 }
 {bone1Name=dyn_hip_gadget_a_r, bone2Name=dyn_hip_gadget_a_base_r, swing1LimitDegrees=15, swing2LimitDegrees=10, twistLimitDegrees=5}
 {
  bone1Name=dyn_hip_gadget_a_base_l
  bone2Name=pelvis
  swing1LimitDegrees=45
  swing1Motion=Free
  swing2LimitDegrees=45
  swing2Motion=Free
  twistLimitDegrees=45
  twistMotion=Free
 }
 {bone1Name=dyn_hip_gadget_a_l, bone2Name=dyn_hip_gadget_a_base_l, swing1LimitDegrees=15, swing2LimitDegrees=15, twistLimitDegrees=10}
 {
  bone1Name=dyn_hip_cableA_a_base_l
  bone2Name=pelvis
  swing1LimitDegrees=45
  swing1Motion=Free
  swing2LimitDegrees=45
  swing2Motion=Free
  twistLimitDegrees=45
  twistMotion=Free
 }
 {bone1Name=dyn_hip_cableA_a_l, bone2Name=dyn_hip_cableA_a_base_l, swing1LimitDegrees=25, swing2LimitDegrees=25, twistLimitDegrees=5}
 {bone1Name=dyn_hip_cableA_b_l, bone2Name=dyn_hip_cableA_a_l, swing1LimitDegrees=25, swing2LimitDegrees=25, twistLimitDegrees=5}
 {
  bone1Name=dyn_hip_cableB_a_base_l
  bone2Name=thigh_l
  swing1LimitDegrees=45
  swing1Motion=Free
  swing2LimitDegrees=45
  swing2Motion=Free
  twistLimitDegrees=45
  twistMotion=Free
 }
 {bone1Name=dyn_hip_cableB_a_l, bone2Name=dyn_hip_cableB_a_base_l, swing1LimitDegrees=45, swing2LimitDegrees=45, twistLimitDegrees=15}
 {bone1Name=dyn_hip_cableB_b_l, bone2Name=dyn_hip_cableB_a_l, swing1LimitDegrees=45, swing2LimitDegrees=30, twistLimitDegrees=15}
 {
  bone1Name=dyn_hip_backClip_a_base_r
  bone2Name=pelvis
  swing1LimitDegrees=45
  swing1Motion=Free
  swing2LimitDegrees=45
  swing2Motion=Free
  twistLimitDegrees=45
  twistMotion=Free
 }
 {
  bone1Name=dyn_hip_backClip_a_r
  bone2Name=dyn_hip_backClip_a_base_r
  swing1LimitDegrees=10
  swing2LimitDegrees=45
  swing2Motion=Locked
  twistLimitDegrees=45
  twistMotion=Locked
 }
 {
  bone1Name=dyn_hip_backClip_b_r
  bone2Name=dyn_hip_backClip_a_r
  swing1LimitDegrees=10
  swing2LimitDegrees=45
  swing2Motion=Locked
  twistLimitDegrees=45
  twistMotion=Locked
 }
 {
  bone1Name=spine_04
  bone2Name=pelvis
  swing1LimitDegrees=45
  swing1Motion=Free
  swing2LimitDegrees=45
  swing2Motion=Free
  twistLimitDegrees=45
  twistMotion=Free
 }
 {
  bone1Name=dyn_harness_a_base_l
  bone2Name=spine_04
  swing1LimitDegrees=45
  swing1Motion=Free
  swing2LimitDegrees=45
  swing2Motion=Free
  twistLimitDegrees=45
  twistMotion=Free
 }
 {bone1Name=dyn_harness_b_l, bone2Name=dyn_harness_a_base_l, swing1LimitDegrees=5, swing2LimitDegrees=5, twistLimitDegrees=5}
 {
  bone1Name=dyn_harness_a_base_r
  bone2Name=spine_04
  swing1LimitDegrees=45
  swing1Motion=Free
  swing2LimitDegrees=45
  swing2Motion=Free
  twistLimitDegrees=45
  twistMotion=Free
 }
 {bone1Name=dyn_harness_b_r, bone2Name=dyn_harness_a_base_r, swing1LimitDegrees=5, swing2LimitDegrees=5, twistLimitDegrees=5}
]
