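# Animation Blueprint for the F_MED_York body cosmetic, copying the attached parent mesh's pose and adding secondary motion.
# Weighted constraints and joint-angle-driven bone offsets correct limb volume and place zippers, harness and hip gear.
# Movement states adjust the left hip cable, then AnimDynamics swings zippers and a final RigidBody pass runs, at half alpha while sprinting.
# EventGraph computes angles and alphas each update, also printing left thigh yaw.
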AnimBlueprint /BRCosmetics/Characters/Player/Female/Medium/Bodies/F_MED_York/Meshes/F_MED_York_AnimBP
target_skeleton=&/F_MED_York_Skeleton
graphs [
 AnimGraph {nodes [
   n1 = Anim Dynamics - Bone: dyn_harness_b_l(Alpha=0, ExternalForce="0,0,0") {
    angularDampingOverride=1
    angularSpringConstant=1500
    bAngularSpring=true
    bGravityOverrideInSimSpace=true
    bOverrideAngularDamping=true
    gravityScale=0.35
    simulationSpace=RootRelative
    alphaBoolBlend {blendOption=Linear}
    boundBone {boneName=dyn_harness_b_l}
    physicsBodyDefinitions [{
      sphereCollisionRadius=0
      boundBone {boneName=dyn_harness_b_l}
      boxExtents=<10 20 20>
      constraintSetup {angularLimitsMax=<0 0 10>, angularLimitsMin=<0 -20 -10>, angularTarget=<0 0 0>}
      localJointOffset=<10 0 0>
     }]
   }
   Anim Dynamics - Bone: dyn_harness_b_r(ComponentPose=@n1, Alpha=0, ExternalForce="0,0,0") {
    angularDampingOverride=1
    angularSpringConstant=1500
    bAngularSpring=true
    bGravityOverrideInSimSpace=true
    bOverrideAngularDamping=true
    gravityScale=0.35
    simulationSpace=RootRelative
    alphaBoolBlend {blendOption=Linear}
    boundBone {boneName=dyn_harness_b_r}
    physicsBodyDefinitions [{
      sphereCollisionRadius=0
      boundBone {boneName=dyn_harness_b_r}
      boxExtents=<10 20 20>
      constraintSetup {angularLimitsMax=<0 0 10>, angularLimitsMin=<0 -20 -10>, angularTarget=<0 0 0>}
      localJointOffset=<-10 0 0>
     }]
   }
   n2 = "Constraint - dyn_volume_base_lowerarm_l (PP)"!(
    ComponentPose=Local To Component(LocalPose=Copy Pose From Mesh() {bUseAttachedParent=true})
    Alpha=1
    ConstraintWeights_0=0.65
    ConstraintWeights_1:float=0.35
   ) -> (Pose) {
    alphaBoolBlend {blendOption=Linear}
    boneToModify {boneName=dyn_volume_base_lowerarm_l}
    constraintSetup [{targetBone {boneName=upperarm_l}}, {targetBone {boneName=lowerarm_l}}]
    constraintWeights [1,1]
   }
   n3 = "Constraint - dyn_volume_base_lowerarm_r (PP)"!(ComponentPose=@n2, Alpha=1, ConstraintWeights_0=0.65, ConstraintWeights_1:float=0.35) -> (Pose) {
    alphaBoolBlend {blendOption=Linear}
    boneToModify {boneName=dyn_volume_base_lowerarm_r}
    constraintSetup [{targetBone {boneName=upperarm_r}}, {targetBone {boneName=lowerarm_r}}]
    constraintWeights [1,1]
   }
   n4 = "Transform (Modify) Bone - Bone: dyn_volume_lowerarm_l"(ComponentPose=@n3, Alpha=Get elbow_rot_l() -> (elbow_rot_l:double), Translation="0,-2,0", Rotation="0,0,0", Scale="0.3,1,1") {
    rotationMode=BMM_Additive
    rotationSpace=BCS_BoneSpace
    scaleMode=BMM_Additive
    scaleSpace=BCS_BoneSpace
    translationMode=BMM_Additive
    translationSpace=BCS_BoneSpace
    alphaBoolBlend {blendOption=Linear}
    alphaScaleBiasClamp {bMapRange=true, inRange {max=80}}
    boneToModify {boneName=dyn_volume_lowerarm_l}
    scale {x=0.3}
    translation {y=-2}
   }
   n5 = "Transform (Modify) Bone - Bone: dyn_volume_lowerarm_r"(ComponentPose=@n4, Alpha=Get elbow_rot_r() -> (elbow_rot_r:double), Translation="0,2,0", Rotation="0,0,0", Scale="0.3,1,1") {
    rotationMode=BMM_Additive
    rotationSpace=BCS_BoneSpace
    scaleMode=BMM_Additive
    scaleSpace=BCS_BoneSpace
    translationMode=BMM_Additive
    translationSpace=BCS_BoneSpace
    alphaBoolBlend {blendOption=Linear}
    alphaScaleBiasClamp {bMapRange=true, inRange {max=80}}
    boneToModify {boneName=dyn_volume_lowerarm_r}
    scale {x=0.3}
    translation {y=2}
   }
   n6 = "Constraint - dyn_volume_base_thigh_l (PP)"!(ComponentPose=@n5, Alpha=1, ConstraintWeights_0=0.6, ConstraintWeights_1:float=0.4) -> (Pose) {
    alphaBoolBlend {blendOption=Linear}
    boneToModify {boneName=dyn_volume_base_thigh_l}
    constraintSetup [{targetBone {boneName=pelvis}}, {targetBone {boneName=thigh_l}}]
    constraintWeights [1,1]
   }
   n7 = "Constraint - dyn_volume_base_thigh_r (PP)"!(ComponentPose=@n6, Alpha=1, ConstraintWeights_0=0.6, ConstraintWeights_1:float=0.4) -> (Pose) {
    alphaBoolBlend {blendOption=Linear}
    boneToModify {boneName=dyn_volume_base_thigh_r}
    constraintSetup [{targetBone {boneName=pelvis}}, {targetBone {boneName=thigh_r}}]
    constraintWeights [1,1]
   }
   n8 = Get thigh_rot_l()
   n9 = "Transform (Modify) Bone - Bone: dyn_volume_thigh_l"(ComponentPose=@n7, Alpha=@n8, Translation="-4,0,0", Rotation="0,0,0", Scale="1,1,1") {
    rotationMode=BMM_Additive
    rotationSpace=BCS_BoneSpace
    scaleMode=BMM_Additive
    scaleSpace=BCS_BoneSpace
    translationMode=BMM_Additive
    translationSpace=BCS_BoneSpace
    alphaBoolBlend {blendOption=Linear}
    alphaScaleBiasClamp {bMapRange=true, inRange {max=0, min=-100}, outRange {max=0, min=1}}
    boneToModify {boneName=dyn_volume_thigh_l}
    translation {x=-4}
   }
   n10 = "Transform (Modify) Bone - Bone: dyn_volume_thigh_l"(ComponentPose=@n9, Alpha=@n8, Translation="0,-4,0", Rotation="0,0,0", Scale="1,1,1") {
    rotationMode=BMM_Additive
    rotationSpace=BCS_BoneSpace
    scaleMode=BMM_Additive
    scaleSpace=BCS_BoneSpace
    translationMode=BMM_Additive
    translationSpace=BCS_BoneSpace
    alphaBoolBlend {blendOption=Linear}
    alphaScaleBiasClamp {bMapRange=true, inRange {max=50}}
    boneToModify {boneName=dyn_volume_thigh_l}
    translation {y=-4}
   }
   n11 = Get thigh_rot_r()
   n12 = "Transform (Modify) Bone - Bone: dyn_volume_thigh_r"(ComponentPose=@n10, Alpha=@n11, Translation="4,0,0", Rotation="0,0,0", Scale="1,1,1") {
    rotationMode=BMM_Additive
    rotationSpace=BCS_BoneSpace
    scaleMode=BMM_Additive
    scaleSpace=BCS_BoneSpace
    translationMode=BMM_Additive
    translationSpace=BCS_BoneSpace
    alphaBoolBlend {blendOption=Linear}
    alphaScaleBiasClamp {bMapRange=true, inRange {max=0, min=-100}, outRange {max=0, min=1}}
    boneToModify {boneName=dyn_volume_thigh_r}
    translation {x=4}
   }
   n13 = "Transform (Modify) Bone - Bone: dyn_volume_thigh_r"(ComponentPose=@n12, Alpha=@n11, Translation="0,4,0", Rotation="0,0,0", Scale="1,1,1") {
    rotationMode=BMM_Additive
    rotationSpace=BCS_BoneSpace
    scaleMode=BMM_Additive
    scaleSpace=BCS_BoneSpace
    translationMode=BMM_Additive
    translationSpace=BCS_BoneSpace
    alphaBoolBlend {blendOption=Linear}
    alphaScaleBiasClamp {bMapRange=true, inRange {max=50}}
    boneToModify {boneName=dyn_volume_thigh_r}
    translation {y=4}
   }
   n14 = "Constraint - dyn_volume_base_calf_l (PP)"!(ComponentPose=@n13, Alpha=1, ConstraintWeights_0=0.5, ConstraintWeights_1:float=0.5) -> (Pose) {
    alphaBoolBlend {blendOption=Linear}
    boneToModify {boneName=dyn_volume_base_calf_l}
    constraintSetup [{targetBone {boneName=thigh_l}}, {targetBone {boneName=calf_l}}]
    constraintWeights [1,1]
   }
   n15 = "Constraint - dyn_volume_base_calf_r (PP)"!(ComponentPose=@n14, Alpha=1, ConstraintWeights_0=0.5, ConstraintWeights_1:float=0.5) -> (Pose) {
    alphaBoolBlend {blendOption=Linear}
    boneToModify {boneName=dyn_volume_base_calf_r}
    constraintSetup [{targetBone {boneName=thigh_r}}, {targetBone {boneName=calf_r}}]
    constraintWeights [1,1]
   }
   n16 = "Transform (Modify) Bone - Bone: dyn_volume_calf_l"(ComponentPose=@n15, Alpha=Get knee_rot_l() -> (knee_rot_l:double), Translation="0,3,0", Rotation="0,0,0", Scale="2,1,1.5") {
    rotationMode=BMM_Additive
    rotationSpace=BCS_BoneSpace
    scaleMode=BMM_Additive
    scaleSpace=BCS_BoneSpace
    translationMode=BMM_Additive
    translationSpace=BCS_BoneSpace
    alphaBoolBlend {blendOption=Linear}
    alphaScaleBiasClamp {bMapRange=true, inRange {max=80}}
    boneToModify {boneName=dyn_volume_calf_l}
    scale {x=2, z=1.5}
    translation {y=3}
   }
   n17 = "Transform (Modify) Bone - Bone: dyn_volume_calf_r"(ComponentPose=@n16, Alpha=Get knee_rot_r() -> (knee_rot_r:double), Translation="0,-3,0", Rotation="0,0,0", Scale="2,1,1.5") {
    rotationMode=BMM_Additive
    rotationSpace=BCS_BoneSpace
    scaleMode=BMM_Additive
    scaleSpace=BCS_BoneSpace
    translationMode=BMM_Additive
    translationSpace=BCS_BoneSpace
    alphaBoolBlend {blendOption=Linear}
    alphaScaleBiasClamp {bMapRange=true, inRange {max=80}}
    boneToModify {boneName=dyn_volume_calf_r}
    scale {x=2, z=1.5}
    translation {y=-3}
   }
   n18 = "Constraint - dyn_chest_zipper_a_base (PP)"!(ComponentPose=@n17, Alpha=1, ConstraintWeights_0=0.1419, ConstraintWeights_1:float=0.8581) -> (Pose) {
    alphaBoolBlend {blendOption=Linear}
    boneToModify {boneName=dyn_chest_zipper_a_base}
    constraintSetup [{targetBone {boneName=spine_04}}, {targetBone {boneName=spine_05}}]
    constraintWeights [1,1]
   }
   n19 = "Constraint - dyn_front_zipper_base_l (PPPP)"!(
    ComponentPose=@n18
    Alpha=1
    ConstraintWeights_0=0.0561
    ConstraintWeights_1:float=0.0528
    ConstraintWeights_2:float=0.8798
    ConstraintWeights_3:float=0.0113
   ) -> (Pose) {
    alphaBoolBlend {blendOption=Linear}
    boneToModify {boneName=dyn_front_zipper_base_l}
    constraintSetup [{targetBone {boneName=spine_02}}, {targetBone {boneName=spine_03}}, {targetBone {boneName=spine_04}}, {targetBone {boneName=spine_05}}]
    constraintWeights [1,1,1,1]
   }
   n20 = "Constraint - dyn_front_zipper_base_r (PPPP)"!(
    ComponentPose=@n19
    Alpha=1
    ConstraintWeights_0=0.0573
    ConstraintWeights_1:float=0.0535
    ConstraintWeights_2:float=0.8785
    ConstraintWeights_3:float=0.0107
   ) -> (Pose) {
    alphaBoolBlend {blendOption=Linear}
    boneToModify {boneName=dyn_front_zipper_base_r}
    constraintSetup [{targetBone {boneName=spine_02}}, {targetBone {boneName=spine_03}}, {targetBone {boneName=spine_04}}, {targetBone {boneName=spine_05}}]
    constraintWeights [1,1,1,1]
   }
   n21 = "Constraint - dyn_harness_a_base_l (PPPPPP)"!(
    ComponentPose=@n20
    Alpha=1
    ConstraintWeights_0=0.2131
    ConstraintWeights_1:float=0.0169
    ConstraintWeights_2:float=0.0187
    ConstraintWeights_3:float=0.4418
    ConstraintWeights_4:float=0.1704
    ConstraintWeights_5:float=0.1392
   ) -> (Pose) {
    alphaBoolBlend {blendOption=Linear}
    boneToModify {boneName=dyn_harness_a_base_l}
    constraintSetup [
     {targetBone {boneName=clavicle_l}}
     {targetBone {boneName=spine_02}}
     {targetBone {boneName=spine_03}}
     {targetBone {boneName=spine_04}}
     {targetBone {boneName=spine_05}}
     {targetBone {boneName=upperarm_twist_01_l}}
    ]
    constraintWeights [1,1,1,1,1,1]
   }
   n22 = "Constraint - dyn_harness_a_base_r (PPPPPP)"!(
    ComponentPose=@n21
    Alpha=1
    ConstraintWeights_0=0.1366
    ConstraintWeights_1:float=0.0169
    ConstraintWeights_2:float=0.0188
    ConstraintWeights_3:float=0.4431
    ConstraintWeights_4:float=0.1709
    ConstraintWeights_5:float=0.2137
   ) -> (Pose) {
    alphaBoolBlend {blendOption=Linear}
    boneToModify {boneName=dyn_harness_a_base_r}
    constraintSetup [
     {targetBone {boneName=upperarm_twist_01_r}}
     {targetBone {boneName=spine_02}}
     {targetBone {boneName=spine_03}}
     {targetBone {boneName=spine_04}}
     {targetBone {boneName=spine_05}}
     {targetBone {boneName=clavicle_r}}
    ]
    constraintWeights [1,1,1,1,1,1]
   }
   n23 = "Constraint - dyn_upleg_zipper_a_r (PP)"!(ComponentPose=@n22, Alpha=1, ConstraintWeights_0=0.6886, ConstraintWeights_1:float=0.3114) -> (Pose) {
    alphaBoolBlend {blendOption=Linear}
    boneToModify {boneName=dyn_upleg_zipper_a_r}
    constraintSetup [{targetBone {boneName=thigh_r}}, {targetBone {boneName=thigh_twist_01_r}}]
    constraintWeights [1,1]
   }
   n24 = "Constraint - dyn_knee_zipper_base_l (PPP)"!(ComponentPose=@n23, Alpha=1, ConstraintWeights_0=0.0514, ConstraintWeights_1:float=0.7923, ConstraintWeights_2:float=0.1563) -> (Pose) {
    alphaBoolBlend {blendOption=Linear}
    boneToModify {boneName=dyn_knee_zipper_base_l}
    constraintSetup [{targetBone {boneName=calf_l}}, {targetBone {boneName=thigh_l}}, {targetBone {boneName=dyn_volume_calf_l}}]
    constraintWeights [1,1,1]
   }
   n25 = Get thigh_rot_l()
   n26 = "Transform (Modify) Bone - Bone: dyn_hip_gadget_a_base_l"(ComponentPose=@n24, Alpha=@n25, Translation="0,-6,-1", Rotation="0,-10,-60", Scale="1,1,1") {
    rotationMode=BMM_Additive
    rotationSpace=BCS_BoneSpace
    scaleMode=BMM_Additive
    scaleSpace=BCS_BoneSpace
    translationMode=BMM_Additive
    translationSpace=BCS_BoneSpace
    alphaBoolBlend {blendOption=Linear}
    alphaScaleBiasClamp {bMapRange=true, inRange {max=0, min=-100}, outRange {max=0, min=1}}
    boneToModify {boneName=dyn_hip_gadget_a_base_l}
    rotation {roll=-60, yaw=-10}
    translation {y=-6, z=-1}
   }
   n27 = "Transform (Modify) Bone - Bone: dyn_hip_cableA_a_base_l"(ComponentPose=@n26, Alpha=@n25, Translation="0,0,0", Rotation="-15,0,-30", Scale="1,1,1") {
    rotationMode=BMM_Additive
    rotationSpace=BCS_BoneSpace
    scaleMode=BMM_Additive
    scaleSpace=BCS_BoneSpace
    translationMode=BMM_Additive
    translationSpace=BCS_BoneSpace
    alphaBoolBlend {blendOption=Linear}
    alphaScaleBiasClamp {bMapRange=true, inRange {max=0, min=-100}, outRange {max=0, min=1}}
    boneToModify {boneName=dyn_hip_cableA_a_base_l}
    rotation {pitch=-15, roll=-30}
   }
   n28 = "Transform (Modify) Bone - Bone: dyn_hip_cableB_a_base_l"(ComponentPose=@n27, Alpha=Get cabel_offset_Y() -> (cabel_offset_Y:float), Translation="0,0,0", Rotation="-50,0,0", Scale="1,1,1") {
    rotationMode=BMM_Additive
    rotationSpace=BCS_BoneSpace
    scaleMode=BMM_Additive
    scaleSpace=BCS_BoneSpace
    translationMode=BMM_Additive
    translationSpace=BCS_BoneSpace
    alphaBoolBlend {blendOption=Linear}
    alphaScaleBiasClamp {bMapRange=true, inRange {max=0, min=-100}, outRange {max=0, min=1}}
    boneToModify {boneName=dyn_hip_cableB_a_base_l}
    rotation {pitch=-50}
   }
   n29 = "Transform (Modify) Bone - Bone: dyn_hip_cableB_a_base_l"!(ComponentPose=@n28, bAlphaBoolEnabled:bool=Get bIsCrouchSprinting(), Translation="0,-5,0", Rotation="-5,-5,0", Scale="1,1,1") -> (Pose) {
    alphaInputType=Bool
    rotationMode=BMM_Additive
    rotationSpace=BCS_BoneSpace
    scaleMode=BMM_Additive
    scaleSpace=BCS_BoneSpace
    translationMode=BMM_Additive
    translationSpace=BCS_BoneSpace
    alphaBoolBlend {blendInTime=0.2, blendOption=Linear, blendOutTime=0.4}
    alphaScaleBiasClamp {bMapRange=true, inRange {max=0, min=-100}, outRange {max=0, min=1}}
    boneToModify {boneName=dyn_hip_cableB_a_base_l}
    rotation {pitch=-5, yaw=-5}
    translation {y=-5}
   }
   n30 = "Transform (Modify) Bone - Bone: dyn_hip_cableB_a_base_l"(ComponentPose=@n29, bAlphaBoolEnabled=Get bIsDBNO(), Translation="0,-10,0", Rotation="-5,-5,0", Scale="1,1,1") {
    alphaInputType=Bool
    rotationMode=BMM_Additive
    rotationSpace=BCS_BoneSpace
    scaleMode=BMM_Additive
    scaleSpace=BCS_BoneSpace
    translationMode=BMM_Additive
    translationSpace=BCS_BoneSpace
    alphaBoolBlend {blendInTime=0.2, blendOption=Linear, blendOutTime=0.4}
    alphaScaleBiasClamp {bMapRange=true, inRange {max=0, min=-100}, outRange {max=0, min=1}}
    boneToModify {boneName=dyn_hip_cableB_a_base_l}
    rotation {pitch=-5, yaw=-5}
    translation {y=-10}
   }
   n31 = "Transform (Modify) Bone - Bone: dyn_hip_cableB_a_base_l"!(ComponentPose=@n30, Alpha:float=Get jumpAlpha() -> (jumpAlpha:double), Translation="0,-2,0", Rotation="0,-40,0", Scale="1,1,1") -> (Pose) {
    rotationMode=BMM_Additive
    rotationSpace=BCS_BoneSpace
    scaleMode=BMM_Additive
    scaleSpace=BCS_BoneSpace
    translationMode=BMM_Additive
    translationSpace=BCS_BoneSpace
    alphaBoolBlend {blendOption=Linear}
    alphaScaleBiasClamp {bInterpResult=true, interpSpeedDecreasing=5, interpSpeedIncreasing=5, inRange {max=0, min=-100}, outRange {max=0, min=1}}
    boneToModify {boneName=dyn_hip_cableB_a_base_l}
    rotation {yaw=-40}
    translation {y=-2}
   }
   n32 = "Transform (Modify) Bone - Bone: dyn_hip_gadget_a_base_r"(ComponentPose=@n31, Alpha=1, Translation="0,-3.5,-6", Rotation="-15,5,60", Scale="1,1,1") {
    rotationMode=BMM_Additive
    rotationSpace=BCS_BoneSpace
    scaleMode=BMM_Additive
    scaleSpace=BCS_BoneSpace
    translationMode=BMM_Additive
    translationSpace=BCS_BoneSpace
    alphaBoolBlend {blendOption=Linear}
    alphaScaleBiasClamp {inRange {max=0, min=-100}, outRange {max=0, min=1}}
    boneToModify {boneName=dyn_hip_gadget_a_base_r}
    rotation=<rot -15 5 60>
    translation {y=-3.5, z=-6}
   }
   n33 = Anim Dynamics - Bone: dyn_chest_zipper_b(ComponentPose=@n32, Alpha=1, ExternalForce="0,0,0") {
    angularDampingOverride=1
    angularSpringConstant=1500
    bAngularSpring=true
    bGravityOverrideInSimSpace=true
    bOverrideAngularDamping=true
    gravityScale=0.35
    simulationSpace=RootRelative
    alphaBoolBlend {blendOption=Linear}
    boundBone {boneName=dyn_chest_zipper_b}
    physicsBodyDefinitions [{
      sphereCollisionRadius=0
      boundBone {boneName=dyn_chest_zipper_b}
      boxExtents=<10 20 20>
      constraintSetup {angularLimitsMax=<0 0 60>, angularLimitsMin=<0 0 0>, angularTarget=<0 0 0>}
      localJointOffset=<10 0 0>
     }]
   }
   n34 = Anim Dynamics - Bone: dyn_front_zipper_l(ComponentPose=@n33, Alpha=1, ExternalForce="0,0,0") {
    angularDampingOverride=1
    angularSpringConstant=1500
    bAngularSpring=true
    bGravityOverrideInSimSpace=true
    bOverrideAngularDamping=true
    gravityScale=0.35
    simulationSpace=RootRelative
    alphaBoolBlend {blendOption=Linear}
    boundBone {boneName=dyn_front_zipper_l}
    physicsBodyDefinitions [{
      sphereCollisionRadius=0
      boundBone {boneName=dyn_front_zipper_l}
      boxExtents=<10 20 20>
      constraintSetup {angularLimitsMax=<0 0 60>, angularLimitsMin=<0 0 0>, angularTarget=<0 0 0>}
      localJointOffset=<10 0 0>
     }]
   }
   n35 = Anim Dynamics - Bone: dyn_front_zipper_base_r(ComponentPose=@n34, Alpha=1, ExternalForce="0,0,0") {
    angularDampingOverride=1
    angularSpringConstant=1500
    bAngularSpring=true
    bGravityOverrideInSimSpace=true
    bOverrideAngularDamping=true
    gravityScale=0.35
    simulationSpace=RootRelative
    alphaBoolBlend {blendOption=Linear}
    boundBone {boneName=dyn_front_zipper_base_r}
    physicsBodyDefinitions [{
      sphereCollisionRadius=0
      boundBone {boneName=dyn_front_zipper_base_r}
      boxExtents=<10 20 20>
      constraintSetup {angularLimitsMax=<0 0 60>, angularLimitsMin=<0 0 0>, angularTarget=<0 0 0>}
      localJointOffset=<-10 0 0>
     }]
   }
   n36 = Anim Dynamics - Bone: dyn_upleg_zipper_b_r(ComponentPose=@n35, Alpha=1, ExternalForce="0,0,0") {
    angularDampingOverride=1
    angularSpringConstant=1500
    bAngularSpring=true
    bGravityOverrideInSimSpace=true
    bOverrideAngularDamping=true
    gravityScale=0.35
    simulationSpace=RootRelative
    alphaBoolBlend {blendOption=Linear}
    boundBone {boneName=dyn_upleg_zipper_b_r}
    physicsBodyDefinitions [{
      sphereCollisionRadius=0
      boundBone {boneName=dyn_upleg_zipper_b_r}
      boxExtents=<10 20 20>
      constraintSetup {angularLimitsMax=<0 0 40>, angularLimitsMin=<0 0 0>, angularTarget=<0 0 0>}
      localJointOffset=<-10 0 0>
     }]
   }
   n37 = Anim Dynamics - Bone: dyn_knee_zipper_base_l(ComponentPose=@n36, Alpha=1, ExternalForce="0,0,0") {
    angularDampingOverride=1
    angularSpringConstant=1500
    bAngularSpring=true
    bGravityOverrideInSimSpace=true
    bOverrideAngularDamping=true
    gravityScale=0.35
    simulationSpace=RootRelative
    alphaBoolBlend {blendOption=Linear}
    boundBone {boneName=dyn_knee_zipper_base_l}
    physicsBodyDefinitions [{
      sphereCollisionRadius=0
      boundBone {boneName=dyn_knee_zipper_base_l}
      boxExtents=<10 20 20>
      constraintSetup {angularLimitsMax=<0 0 40>, angularLimitsMin=<0 0 0>, angularTarget=<0 0 0>}
      localJointOffset=<10 0 0>
     }]
   }
   n38 = "Transform (Modify) Bone - Bone: dyn_hip_cableB_a_base_l"(ComponentPose=@n37, Alpha=1, Translation="0,-2,0", Rotation="0,-5,0", Scale="1,1,1") {
    rotationMode=BMM_Additive
    rotationSpace=BCS_BoneSpace
    scaleMode=BMM_Additive
    scaleSpace=BCS_BoneSpace
    translationMode=BMM_Additive
    translationSpace=BCS_BoneSpace
    alphaBoolBlend {blendOption=Linear}
    alphaScaleBiasClamp {inRange {max=0, min=-100}, outRange {max=0, min=1}}
    boneToModify {boneName=dyn_hip_cableB_a_base_l}
    rotation {yaw=-5}
    translation {y=-2}
   }
   n39 = RigidBody(ComponentPose=@n38, Alpha=Get Rigid Body Alpha() -> (Rigid Body Alpha:float)) {
    bClampLinearTranslationLimitToRefPose=true
    bOverrideWorldGravity=true
    overridePhysicsAsset=&/F_MED_York_Physics
    simulationSpace=BaseBoneSpace
    alphaBoolBlend {blendOption=Linear}
    baseBoneRef {boneName=pelvis}
    componentLinearAccScale=<0.35 0.35 0.35>
    componentLinearVelScale=<0.35 0.35 0.35>
    overrideWorldGravity {z=-980}
   }
   Output Pose(Result=Component To Local(ComponentPose=@n39)) {layerGroup=None}
   Get thigh_rot_r()
  ]}
 EventGraph {nodes [
   Event BlueprintUpdateAnimation()
   > n1 = Sequence!(execute) -> (then_0, then_1, then_2:Exec, then_3:Exec, then_4:Exec, then_5:Exec, then_6:Exec, then_7:Exec)
   n2 = Branch(execute=@n1.then_6, Condition=OR Boolean(A=Get bIsJumping(), B=Get bIsFalling()))
   > Set jumpAlpha(jumpAlpha:double=1) -> (Output_Get:double)
   Set jumpAlpha(execute=@n2.else, jumpAlpha=0)
   n3 = Branch(execute=@n1.then_7, Condition=OR Boolean(A=Get bIsSprinting(), B=Get bIsTacticalSprinting()))
   > Set Rigid Body Alpha(Rigid Body Alpha:float=0.5) -> (Output_Get:float)
   Set Rigid Body Alpha(execute=@n3.else, Rigid Body Alpha=1)
   n4 = GetDeltaTransformFromRefPose(self=GetOwnerSkeletalMeshComponent(), BoneName=lowerarm_l, BaseName=upperarm_l)
   Set elbow_rot_l(execute=@n1.then_0, elbow_rot_l:double=BreakRotator(InRot=BreakTransform(InTransform=@n4).Rotation).Yaw) -> (Output_Get:double)
   n5 = GetOwnerSkeletalMeshComponent()
   n6 = BreakRotator(InRot=BreakTransform(InTransform=GetDeltaTransformFromRefPose(self=@n5, BoneName=thigh_r, BaseName=pelvis)).Rotation)
   Set thigh_rot_r(execute=@n1.then_3, thigh_rot_r=@n6.Yaw)
   > Set cabel_offset_Y(cabel_offset_Y:float=@n6.Pitch) -> (Output_Get:float)
   n7 = GetDeltaTransformFromRefPose(self=GetOwnerSkeletalMeshComponent(), BoneName=calf_r, BaseName=thigh_r)
   Set knee_rot_r(execute=@n1.then_5, knee_rot_r:double=BreakRotator(InRot=BreakTransform(InTransform=@n7).Rotation).Yaw) -> (Output_Get:double)
   n8 = GetDeltaTransformFromRefPose(self=GetOwnerSkeletalMeshComponent(), BoneName=calf_l, BaseName=thigh_l)
   Set knee_rot_l(execute=@n1.then_4, knee_rot_l:double=BreakRotator(InRot=BreakTransform(InTransform=@n8).Rotation).Yaw) -> (Output_Get:double)
   n9 = GetDeltaTransformFromRefPose(self=GetOwnerSkeletalMeshComponent(), BoneName=thigh_l, BaseName=pelvis)
   n10 = Set thigh_rot_l(execute=@n1.then_2, thigh_rot_l=BreakRotator(InRot=BreakTransform(InTransform=@n9).Rotation).Yaw)
   > PrintString!(execute, InString="To String (Float)"(InDouble=@n10), bPrintToScreen, bPrintToLog, TextColor, Duration=0, Key) -> (then)
   n11 = GetDeltaTransformFromRefPose(self=GetOwnerSkeletalMeshComponent(), BoneName=lowerarm_r, BaseName=upperarm_r)
   Set elbow_rot_r(execute=@n1.then_1, elbow_rot_r:double=BreakRotator(InRot=BreakTransform(InTransform=@n11).Rotation).Yaw) -> (Output_Get:double)
  ]}
]
variables [
 cabel_offset_Y {type=number}
 elbow_rot_l {type=number}
 elbow_rot_r {type=number}
 jumpAlpha {type=number}
 knee_rot_l {type=number}
 knee_rot_r {type=number}
 Rigid Body Alpha {type=number}
 thigh_rot_l {type=number}
 thigh_rot_r {type=number}
]
